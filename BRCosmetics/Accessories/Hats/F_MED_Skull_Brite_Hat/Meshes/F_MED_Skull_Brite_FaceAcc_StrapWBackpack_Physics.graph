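# Physics asset for the Skull Brite hat's face-accessory straps (backpack variant), defining bodies and joints for two dangling strap chains.
# Each five-capsule chain branches from a head-mounted driver bone, with limited swing and twist that stiffens toward thicker end segments.
# Both chain ends also link to spine_05 through free-rotation constraints, closing loops between head and upper torso.
# Head, neck_02 and spine_05 carry large body capsules, while zero-size capsules serve as linking bodies.
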PhysicsAsset /BRCosmetics/Accessories/Hats/F_MED_Skull_Brite_Hat/Meshes/F_MED_Skull_Brite_FaceAcc_StrapWBackpack_Physics
bodies {
 dyn_strapA_01 [{shapeType=Capsule, length=5, radius=2, shapeName=dyn_strapA_01_capsule, center=<3.4567 0 0>, rotation=<rot 90 0 0>}]
 dyn_strapA_02 [{shapeType=Capsule, length=5, radius=2, shapeName=dyn_strapA_02_capsule, center=<3.4567 0 0>, rotation=<rot 90 0 0>}]
 dyn_strapA_03 [{shapeType=Capsule, length=4.4736, radius=2.5264, shapeName=dyn_strapA_03_capsule, center=<3.4567 0 0>, rotation=<rot 90 0 0>}]
 dyn_strapA_04 [{shapeType=Capsule, length=4.1618, radius=2.8382, shapeName=dyn_strapA_04_capsule, center=<3.4567 0 0>, rotation=<rot 90 0 0>}]
 dyn_strapA_05 [{shapeType=Capsule, length=3.8682, radius=3.1318, shapeName=dyn_strapA_05_capsule, center=<3.4567 0 0>, rotation=<rot 90 0 0>}]
 dyn_strapB_01 [{shapeType=Capsule, length=5, radius=2, shapeName=dyn_strapB_01_capsule, center=<3.4567 0 0>, rotation=<rot 90 0 0>}]
 dyn_strapB_02 [{shapeType=Capsule, length=5, radius=2, shapeName=dyn_strapB_02_capsule, center=<3.4567 0 0>, rotation=<rot 90 0 0>}]
 dyn_strapB_03 [{shapeType=Capsule, length=5, radius=2, shapeName=dyn_strapB_03_capsule, center=<3.4567 0 0>, rotation=<rot 90 0 0>}]
 dyn_strapB_04 [{shapeType=Capsule, length=4.7679, radius=2.2321, shapeName=dyn_strapB_04_capsule, center=<3.4567 0 0>, rotation=<rot 90 0 0>}]
 dyn_strapB_05 [{shapeType=Capsule, length=4.5675, radius=2.4325, shapeName=dyn_strapB_05_capsule, center=<3.4567 0 0>, rotation=<rot 90 0 0>}]
 dyn_strap_driver [{shapeType=Capsule, length=0, radius=0, shapeName=dyn_strap_driver_capsule}]
 dyn_strap_parent [{shapeType=Capsule, length=0, radius=0, shapeName=dyn_strap_parent_capsule}]
 head [{shapeType=Capsule, length=26.0755, radius=7.228, shapeName=head_capsule, center=<1.1517 -2.125 0>}]
 neck_01 [{shapeType=Capsule, length=0, radius=0, shapeName=neck_01_capsule}]
 neck_02 [{shapeType=Capsule, length=27.6002, radius=7.228, shapeName=head_capsule, center=<-0.7726 -0.7645 0>}]
 spine_04 [{shapeType=Capsule, length=0, radius=0, shapeName=spine_04_capsule}]
 spine_05 [{shapeType=Capsule, length=27.0347, radius=9.8001, shapeName=head_capsule, center=<2.3085 1.36 0>}]
}
constraints [
 {
  bone1Name=spine_05
  bone2Name=spine_04
  swing1LimitDegrees=45
  swing1Motion=Free
  swing2LimitDegrees=45
  swing2Motion=Free
  twistLimitDegrees=45
  twistMotion=Free
 }
 {
  bone1Name=neck_01
  bone2Name=spine_05
  swing1LimitDegrees=45
  swing1Motion=Free
  swing2LimitDegrees=45
  swing2Motion=Free
  twistLimitDegrees=45
  twistMotion=Free
 }
 {
  bone1Name=neck_02
  bone2Name=neck_01
  swing1LimitDegrees=45
  swing1Motion=Free
  swing2LimitDegrees=45
  swing2Motion=Free
  twistLimitDegrees=45
  twistMotion=Free
 }
 {
  bone1Name=head
  bone2Name=neck_02
  swing1LimitDegrees=45
  swing1Motion=Free
  swing2LimitDegrees=45
  swing2Motion=Free
  twistLimitDegrees=45
  twistMotion=Free
 }
 {
  bone1Name=dyn_strap_parent
  bone2Name=head
  swing1LimitDegrees=45
  swing1Motion=Free
  swing2LimitDegrees=45
  swing2Motion=Free
  twistLimitDegrees=45
  twistMotion=Free
 }
 {
  bone1Name=dyn_strap_driver
  bone2Name=dyn_strap_parent
  swing1LimitDegrees=45
  swing1Motion=Free
  swing2LimitDegrees=45
  swing2Motion=Free
  twistLimitDegrees=45
  twistMotion=Free
 }
 {bone1Name=dyn_strapA_01, bone2Name=dyn_strap_driver, swing1LimitDegrees=20, swing2LimitDegrees=20, twistLimitDegrees=0}
 {bone1Name=dyn_strapA_05, bone2Name=dyn_strapA_04, swing1LimitDegrees=5, swing2LimitDegrees=10, twistLimitDegrees=0}
 {bone1Name=dyn_strapA_04, bone2Name=dyn_strapA_03, swing1LimitDegrees=5, swing2LimitDegrees=10, twistLimitDegrees=5}
 {bone1Name=dyn_strapA_03, bone2Name=dyn_strapA_02, swing1LimitDegrees=20, swing2LimitDegrees=20, twistLimitDegrees=5}
 {bone1Name=dyn_strapA_02, bone2Name=dyn_strapA_01, swing1LimitDegrees=20, swing2LimitDegrees=20, twistLimitDegrees=5}
 {bone1Name=dyn_strapB_01, bone2Name=dyn_strap_driver, swing1LimitDegrees=20, swing2LimitDegrees=20, twistLimitDegrees=0}
 {bone1Name=dyn_strapB_02, bone2Name=dyn_strapB_01, swing1LimitDegrees=20, swing2LimitDegrees=20, twistLimitDegrees=0}
 {bone1Name=dyn_strapB_03, bone2Name=dyn_strapB_02, swing1LimitDegrees=20, swing2LimitDegrees=20, twistLimitDegrees=5}
 {bone1Name=dyn_strapB_04, bone2Name=dyn_strapB_03, swing1LimitDegrees=5, swing2LimitDegrees=10, twistLimitDegrees=5}
 {bone1Name=dyn_strapB_05, bone2Name=dyn_strapB_04, swing1LimitDegrees=5, swing2LimitDegrees=10, twistLimitDegrees=0}
 {
  bone1Name=spine_05
  bone2Name=dyn_strapA_05
  swing1LimitDegrees=45
  swing1Motion=Free
  swing2LimitDegrees=45
  swing2Motion=Free
  twistLimitDegrees=45
  twistMotion=Free
 }
 {
  bone1Name=dyn_strapB_05
  bone2Name=spine_05
  swing1LimitDegrees=45
  swing1Motion=Free
  swing2LimitDegrees=45
  swing2Motion=Free
  twistLimitDegrees=45
  twistMotion=Free
 }
]
