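# Post-process style animation Blueprint for the Skull Brite hat/face accessory, driving its straps and hair tips on the player.
# It caches a pose copied from the attached parent, blended with body-mesh limb bones.
# Chained RigidBody nodes simulate straps, choosing physics assets by backbling presence, then hair tips at partial alpha.
# With rigid bodies disabled, output switches to Copy Bone and Look At strap posing around spine_05.
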AnimBlueprint /BRCosmetics/Accessories/Hats/F_MED_Skull_Brite_Hat/Meshes/F_MED_Skull_Brite_FaceAcc_AnimBP
target_skeleton=/Game/Accessories/Hats/F_MED_Skull_Brite_Hat/Meshes/F_MED_Skull_Brite_FaceAcc_Skeleton
graphs [
 AnimGraph {nodes [
   n1 = RigidBody!(
    ComponentPose=Local To Component(LocalPose=Use cached pose 'AnimCache'())
    bAlphaBoolEnabled:bool=Get IsWearingBackbling() -> (IsWearingBackbling:bool)
    OverrideWorldGravity:Vector="0,0,-980"
   ) -> (Pose) {
    alphaInputType=Bool
    bClampLinearTranslationLimitToRefPose=true
    bOverrideWorldGravity=true
    overridePhysicsAsset=&/F_MED_Skull_Brite_FaceAcc_StrapWBackpack_Physics
    simulationSpace=BaseBoneSpace
    alphaBoolBlend {blendOption=Linear}
    baseBoneRef {boneName=spine_05}
    simSpaceSettings {maxAngularAcceleration=400, maxAngularVelocity=3, maxLinearAcceleration=500, maxLinearVelocity=250, velocityScaleZ=0, worldAlpha=1}
   }
   n2 = RigidBody!(ComponentPose=@n1, Alpha:float=Get NotBackbling() -> (NotBackbling:double), OverrideWorldGravity="0,0,-980") -> (Pose) {
    bClampLinearTranslationLimitToRefPose=true
    bOverrideWorldGravity=true
    overridePhysicsAsset=&/F_MED_Skull_Brite_FaceAcc_Strap_Physics
    simulationSpace=BaseBoneSpace
    alphaBoolBlend {blendOption=Linear}
    baseBoneRef {boneName=spine_05}
    componentAppliedLinearAccClamp=<4000 4000 4000>
    componentLinearAccScale=<0.3 0.2 0.2>
    simSpaceSettings {
     maxAngularAcceleration=400
     maxAngularVelocity=3
     maxLinearAcceleration=500
     maxLinearVelocity=250
     velocityScaleZ=0.5
     worldAlpha=1
     externalAngularVelocity=<0.7 0.7 0.3>
     externalLinearDragV=<0.4 0.4 0.4>
     externalLinearVelocity=<0.7 0.7 0.3>
    }
   }
   n3 = RigidBody!(ComponentPose=@n2, Alpha=0.4, ExternalForce:Vector) -> (Pose) {
    bClampLinearTranslationLimitToRefPose=true
    bOverrideWorldGravity=true
    overridePhysicsAsset=/Game/Accessories/Hats/F_MED_Skull_Brite_Hat/Meshes/F_MED_Skull_Brite_HairTips_Physics
    simulationSpace=BaseBoneSpace
    alphaBoolBlend {blendOption=Linear}
    baseBoneRef {boneName=root}
    componentAppliedLinearAccClamp=<3000 3000 3000>
    componentLinearAccScale=<0.5 0.5 0.5>
    componentLinearVelScale=<0.5 0.5 0.5>
    overrideWorldGravity {z=-980}
   }
   n4 = Copy Bone - Source Bone: dyn_strapB_08 - Target Bone: spine_05!(
    ComponentPose:Component Space Pose Link Structure=Local To Component(LocalPose=Use cached pose 'AnimCache'())
    Alpha:float=1
    bCopyTranslation:bool=true
    bCopyRotation:bool=true
    bCopyScale:bool=False
   ) -> (Pose:Component Space Pose Link Structure) {
    is=title AnimGraphNode_CopyBone
    controlSpace=BCS_BoneSpace
    alphaBoolBlend {blendOption=Linear}
    sourceBone {boneName=dyn_strapB_08}
    targetBone {boneName=spine_05}
   }
   n5 = Copy Bone - Source Bone: dyn_strapA_09 - Target Bone: spine_05!(
    ComponentPose:Component Space Pose Link Structure=@n4
    Alpha:float=1
    bCopyTranslation:bool=true
    bCopyRotation:bool=true
    bCopyScale:bool=False
   ) -> (Pose:Component Space Pose Link Structure) {
    is=title AnimGraphNode_CopyBone
    controlSpace=BCS_BoneSpace
    alphaBoolBlend {blendOption=Linear}
    sourceBone {boneName=dyn_strapA_09}
    targetBone {boneName=spine_05}
   }
   n6 = Look At - Bone: dyn_strapB_02(ComponentPose=@n5, Alpha=1) {
    alphaBoolBlend {blendOption=Linear}
    boneToModify {boneName=dyn_strapB_02}
    lookAtLocation=<-5 -15 -5>
    lookAtTarget {boneReference {boneName=spine_05}}
    lookAt_Axis {axis=<1 0>}
    lookUp_Axis {axis=<0 1>}
   }
   n7 = Look At - Bone: dyn_strapA_02(ComponentPose=@n6, Alpha=1) {
    alphaBoolBlend {blendOption=Linear}
    boneToModify {boneName=dyn_strapA_02}
    lookAtLocation=<-5 -14 2>
    lookAtTarget {boneReference {boneName=spine_05}}
    lookAt_Axis {axis=<1 0>}
    lookUp_Axis {axis=<0 1>}
   }
   Output Pose(Result=Blend Poses by bool(
     BlendPose_0=Component To Local(ComponentPose=@n7)
     BlendPose_1=Component To Local(ComponentPose=@n3)
     bActiveValue=Get IsRigidBodyOff() -> (IsRigidBodyOff:bool)
    ) {blendType=Linear}) {layerGroup=None}
   n8 = Layered blend per bone(BasePose=Copy Pose From Mesh() {bUseAttachedParent=true}, BlendPoses_0=Copy Pose From Mesh(SourceMeshComponent=Get Body_Mesh())) {layerSetup [{branchFilters [
       {boneName=upperarm_l}
       {boneName=upperarm_r}
       {boneName=lowerarm_l}
       {boneName=lowerarm_r}
       {boneName=thigh_l}
       {boneName=thigh_r}
       {boneName=calf_l}
       {boneName=calf_r}
      ]}]}
   Save cached pose 'AnimCache'(Pose=@n8) {cacheName=AnimCache}
  ]}
 EventGraph {nodes [
   Event BlueprintUpdateAnimation()
   > n1 = Sequence()
   > n2 = Gravity Override State Function DEPRECATED(CharacterPartAnimInstance=Self-Reference(), GameTimeInSeconds=GetGameTimeInSeconds(), GravityOverrideParams=Get Gravity Override Params())
   > Set Gravity Override(Gravity Override=@n2)
   > Set IsRigidBodyOff(IsRigidBodyOff:bool=NOT Boolean(A=Get bRigidBodyEnabled())) -> (Output_Get:bool)
   n3 = CheckForBackbling()
   n4 = Set IsWearingBackbling(execute=@n1.then_1, IsWearingBackbling:bool=@n3) -> (Output_Get:bool)
   > Set Is Backbling A Cape(Is Backbling A Cape:bool=@n3.bIsBackblingACape) -> (Output_Get:bool)
   > Set NotBackbling(NotBackbling:double="To Float (Boolean)"(InBool=NOT Boolean(A=@n3))) -> (Output_Get:double)
   > Set is Wear a Cape(is Wear a Cape:bool=AND Boolean(A=@n3, B=@n3.bIsBackblingLong)) -> (Output_Get:bool)
   > n5 = Branch(Condition=@n4)
   > n6 = Branch(Condition=@n3.bIsBackblingLong)
   Set IsShortBackblingOnAlpha(execute=[@n6.then, @n5.else], IsShortBackblingOnAlpha:double=0) -> (Output_Get:double)
   Set IsShortBackblingOnAlpha(execute=@n6.else, IsShortBackblingOnAlpha=1)
   n7 = Event OnCharacterCustomizationLoaded()
   Sequence(execute=[@n7.then, @n11.then])
   > Set Body_Mesh(Body_Mesh=GetOwnerSkeletalMeshComponent())
   n8 = TryGetPawnOwner()
   > Is Valid(InputObject=@n8)
   > n9 = Cast To FortPlayerPawnAthena(Object=@n8)
   > Set As Fort Player Pawn Athena(As Fort Player Pawn Athena:Fort Player Pawn Athena Object Reference=@n9) -> (Output_Get:Fort Player Pawn Athena Object Reference)
   > n10 = Cast To PlayerPawn_Athena(Object=@n8)
   > Set Player Pawn(Player Pawn:Player Pawn Athena Object Reference=@n10) -> (Output_Get:Player Pawn Athena Object Reference)
   n11 = Event BlueprintInitializeAnimation()
  ]}
]
variables [
 addon_vector {type=Vector}
 As Fort Player Pawn Athena {type=FortPlayerPawnAthena}
 Body_Mesh {type=SkeletalMeshComponent}
 BRAlpha {type=number}
 diving {type=number}
 Gravity Override {type=Vector}
 Gravity Override Params {type=GravityOverrideParamsStruct}
 Is Backbling A Cape {type=boolean}
 Is LOD1/RBAN off/Wearing Backpack {type=boolean}
 is Wear a Cape {type=boolean}
 Is Wearing Backpack {type=boolean}
 IsRigidBodyOff {type=boolean}
 IsShortBackblingOnAlpha {type=number}
 IsWearingBackbling {type=boolean}
 NotBackbling {type=number}
 parachute {type=boolean}
 parent {type=boolean}
 parent_object {type=PlayerPawn_Athena_C}
 physics_vector {type=Vector}
 Player Pawn {type=PlayerPawn_Athena_C}
]
